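# Conversation database for the Guild NPC, who presents as a treasure hunter for hire.
# The Guild entry tag leads to one speech node that holds a default line plus prioritized contextual greetings for a first-ever conversation and for specific player outfits.
# The speech then branches through link nodes into the shared NPC library Quests and BasicServices conversations, so this asset defines only the character's greeting.
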ConversationDatabase /NPCLibrary/NPCs/Guild/NPCLibrary_Conversation_Guild
entry_points [{tag=Conversation.Entry.Guild, destinations [#1]}]
nodes [
 ConversationEntryPointNode_7 {
  id=#1
  class=ConversationEntryPointNode
  outputs [#2]
  properties {entryTag {tagName=Conversation.Entry.Guild}}
 }
 Speech_C_4 {
  id=#2
  class=Speech_C
  outputs [#3, #4]
  properties {
   nodeName=""
   generalConfig {
    defaultMessage=Got a caper? A scheme? A job? A mess? I'm your guy.
    contextualMessages [
     {
      message="Hey. If you need a treasure hunter, I'm one of the best."
      contextRequirements [{requirement {class=FortControllerRequirement_IsFirstEverConversationWithNPC}}]
      priority {curve {rowName=Tandem.Speech.Priority.FirstEverInteraction}}
     }
     {
      message=Got a caper? A scheme? A job? A mess? I'm your guy.
      requirementMatchPolicy=RequireAny
      contextRequirements []
      priority {curve {curveTable=null, rowName=Tandem.Speech.Priority.OutfitReaction}}
     }
     {
      message=You can't trust people in this business. Everyone's out for themselves.
      requirementMatchPolicy=RequireAny
      contextRequirements [{requirement {
         class=FortControllerRequirement_HasCID
         properties {allowedCIds [/BRCosmetics/Athena/Items/Cosmetics/Characters/CID_397_Athena_Commando_F_TreasureHunterFashion]}
        }}]
      priority {curve {rowName=Tandem.Speech.Priority.OutfitReaction}}
     }
     {
      message=I don't like it when people mess with Aura.
      requirementMatchPolicy=RequireAny
      contextRequirements [{requirement {class=FortControllerRequirement_HasCID, properties {allowedCIds [/Game/Athena/Items/Cosmetics/Characters/CID_206_Athena_Commando_M_Bling]}}}]
      priority {curve {rowName=Tandem.Speech.Priority.OutfitReaction}}
     }
    ]
   }
  }
 }
 ConversationLinkNode_0 {id=#4, class=ConversationLinkNode, properties {remoteEntryTag {tagName=Conversation.Entry.BasicServices}}}
 ConversationLinkNode_1 {id=#3, class=ConversationLinkNode, properties {remoteEntryTag {tagName=Conversation.Entry.NPCLibrary.Quests}}}
]
